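# Anim Blueprint for the SteedShimmer backpack, offsetting the pack by the wearer's spine bend and adding secondary motion.
# AnimDynamics drives body and cords into a cached pose; a bool blend adds part_up/part_down bounce except during elimination reactions.
# Enemy-elimination events, or a repeating front-end trigger, briefly set that bool and restart curve-based jaw/head rotation values.
# Clap, laugh and sine-wobble logic sits in disconnected chains.
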AnimBlueprint /BRCosmetics/Accessories/FORT_Backpacks/Backpack_F_MED_SteedShimmer/Meshes/F_MED_SteedShimmer_Pack_AnimBP
target_skeleton=&/F_MED_SteedShimmer_Pack_Skeleton
graphs [
 AnimGraph {nodes [
   n1 = "Transform (Modify) Bone - Bone: part_down"(Alpha=Get Clap Alpha() -> (Clap Alpha:double), Translation="0,0,0", Rotation=Get Jaw Laugh Rot() -> (Jaw Laugh Rot:Rotator), Scale="1,1,1") {
    rotationMode=BMM_Additive
    rotationSpace=BCS_BoneSpace
    translationMode=BMM_Additive
    translationSpace=BCS_BoneSpace
    boneToModify {boneName=part_down}
   }
   n2 = "Transform (Modify) Bone - Bone: part_up"(ComponentPose=@n1, Alpha=Get Clap Alpha(), Translation="0,0,0", Rotation=Get Jaw Laugh Rot 2() -> (Jaw Laugh Rot 2:Rotator), Scale="1,1,1") {
    rotationMode=BMM_Additive
    rotationSpace=BCS_BoneSpace
    translationMode=BMM_Additive
    translationSpace=BCS_BoneSpace
    boneToModify {boneName=part_up}
   }
   n3 = "Transform (Modify) Bone - Bone: part_down"(ComponentPose=@n2, Alpha=1, Translation="0,0,0.2", Rotation="0,0,0", Scale="1,1,1") {translationMode=BMM_Additive, translationSpace=BCS_BoneSpace, boneToModify {boneName=part_down}, translation {z=0.2}}
   n4 = Anim Dynamics - Bone: body(
    ComponentPose="Transform (Modify) Bone - Bone: body"(
     ComponentPose="Transform (Modify) Bone - Bone: part_up"(ComponentPose=@n3, Alpha=1, Translation="0,0,1.2", Rotation="0,0,0", Scale="1,1,1") {translationMode=BMM_Additive, translationSpace=BCS_BoneSpace, boneToModify {boneName=part_up}, translation {z=1.2}}
     Alpha=1
     Translation="0,0,5"
     Rotation="0,0,0"
     Scale="1,1,1"
    ) {translationMode=BMM_Additive, translationSpace=BCS_BoneSpace, boneToModify {boneName=body}, translation {z=5}}
    Alpha=1
    ExternalForce="0,0,0"
   ) {
    angularDampingOverride=0.9999
    bLinearSpring=true
    bOverrideAngularDamping=true
    bOverrideLinearDamping=true
    bUseGravityOverride=true
    linearDampingOverride=0.99
    linearSpringConstant=100
    simulationSpace=BoneRelative
    boundBone {boneName=body}
    componentAppliedLinearAccClamp=<3000 3000 3000>
    componentLinearAccScale {z=0.05}
    componentLinearVelScale {z=0.05}
    gravityOverride {z=-600}
    physicsBodyDefinitions [{
      boundBone {boneName=body}
      boxExtents=<10 10 10>
      constraintSetup {angularLimitsMax=<0 0 0>, angularLimitsMin=<0 0 0>, angularTarget=<0 0 0>, linearAxesMax=<10 0 10>, linearAxesMin=<-10 0 -10>}
      localJointOffset=<0.1 0 0.1>
     }]
    relativeSpaceBone {boneName=root}
    simSpaceSettings {maxAngularAcceleration=21, maxAngularVelocity=9}
   }
   n5 = Anim Dynamics - Bone: part_down(ComponentPose=@n4, Alpha=1, ExternalForce="0,0,0") {
    bOverrideLinearDamping=true
    bUseGravityOverride=true
    linearDampingOverride=1
    linearSpringConstant=800
    simulationSpace=RootRelative
    boundBone {boneName=part_down}
    componentLinearAccScale {z=0.05}
    componentLinearVelScale {z=0.05}
    gravityOverride {z=100}
    physicsBodyDefinitions [{
      boundBone {boneName=part_down}
      boxExtents=<10 10 10>
      constraintSetup {angularLimitsMax=<0 0 0>, angularLimitsMin=<0 0 0>, angularTarget=<0 0 0>, linearAxesMin=<0 0 -0.8>}
      localJointOffset=<0 0 0.1>
     }]
   }
   n6 = Anim Dynamics - Bone: part_up(ComponentPose=@n5, Alpha=1, ExternalForce="0,0,0") {
    bLinearSpring=true
    bOverrideLinearDamping=true
    bUseGravityOverride=true
    linearDampingOverride=1
    linearSpringConstant=400
    simulationSpace=RootRelative
    boundBone {boneName=part_up}
    componentAppliedLinearAccClamp=<3003 3000 30000>
    componentLinearAccScale {z=0.4}
    componentLinearVelScale {z=0.3}
    gravityOverride {z=-600}
    physicsBodyDefinitions [{
      boundBone {boneName=part_up}
      boxExtents=<10 10 10>
      constraintSetup {angularLimitsMax=<0 0 0>, angularLimitsMin=<0 0 0>, angularTarget=<0 0 0>, linearAxesMin=<0 0 -1.2>}
      localJointOffset=<0 0 0.1>
     }]
    simSpaceSettings {maxAngularAcceleration=21, maxAngularVelocity=7}
   }
   n7 = Anim Dynamics - Bone: cord_01(ComponentPose=@n6, Alpha=1, ExternalForce="0,0,0") {
    angularDampingOverride=1
    angularSpringConstant=550
    bAngularSpring=true
    bOverrideAngularDamping=true
    bUseGravityOverride=true
    linearDampingOverride=0.99
    linearSpringConstant=550
    worldSpaceMinimumScale=0.1
    boundBone {boneName=cord_01}
    componentAppliedLinearAccClamp=<3000 3000 6000>
    componentLinearAccScale=<0.3 0.3 1>
    componentLinearVelScale=<0.2 0.2 0.1>
    gravityOverride {z=-980}
    physicsBodyDefinitions [{
      boundBone {boneName=cord_01}
      boxExtents=<10 10 10>
      constraintSetup {angularLimitsMax=<0 40 20>, angularLimitsMin=<0 -40 -20>, angularTarget=<1 0 0>}
      localJointOffset=<1 0.1 0.1>
     }]
   }
   n8 = Anim Dynamics - Bone: cord_02(ComponentPose=@n7, Alpha=1, ExternalForce="0,0,0") {
    angularDampingOverride=1
    angularSpringConstant=550
    bAngularSpring=true
    bOverrideAngularDamping=true
    bUseGravityOverride=true
    linearDampingOverride=0.99
    linearSpringConstant=550
    worldSpaceMinimumScale=0.1
    boundBone {boneName=cord_02}
    componentAppliedLinearAccClamp=<3000 3000 6000>
    componentLinearAccScale=<0.3 0.3 1>
    componentLinearVelScale=<0.2 0.2 0.1>
    gravityOverride {z=-980}
    physicsBodyDefinitions [{
      boundBone {boneName=cord_02}
      boxExtents=<10 10 10>
      constraintSetup {angularLimitsMax=<0 20 20>, angularLimitsMin=<0 -20 -20>, angularTarget=<1 0 0>}
      localJointOffset=<1 0.1 0.1>
     }]
   }
   Component To Local(ComponentPose=@n8)
   n9 = "Transform (Modify) Bone - Bone: part_down"(
    ComponentPose="Transform (Modify) Bone - Bone: part_up"(Alpha=1, Translation="0,0,0.2", Rotation="0,0,0", Scale="1,1,1") {translationMode=BMM_Additive, translationSpace=BCS_BoneSpace, boneToModify {boneName=part_up}, translation {z=0.2}}
    Alpha=1
    Translation="0,0,-0.2"
    Rotation="0,0,0"
    Scale="1,1,1"
   ) {translationMode=BMM_Additive, translationSpace=BCS_BoneSpace, boneToModify {boneName=part_down}, translation {z=-0.2}}
   n10 = "Transform (Modify) Bone - Bone: part_up"!(ComponentPose=@n9, Alpha=1, Translation=Get body move_up() -> (body move_up:Vector)) -> (Pose) {translationMode=BMM_Additive, translationSpace=BCS_BoneSpace, alphaBoolBlend {blendOption=Linear}, boneToModify {boneName=part_up}}
   n11 = "Transform (Modify) Bone - Bone: part_down"!(ComponentPose=@n10, Alpha=1, Translation=Get body move_down() -> (body move_down:Vector)) -> (Pose) {translationMode=BMM_Additive, translationSpace=BCS_BoneSpace, alphaBoolBlend {blendOption=Linear}, boneToModify {boneName=part_down}}
   n12 = Anim Dynamics - Bone: cord_01(ComponentPose=@n11, Alpha=1, ExternalForce="0,0,0") {
    angularDampingOverride=1
    angularSpringConstant=550
    bAngularSpring=true
    bOverrideAngularDamping=true
    bUseGravityOverride=true
    linearDampingOverride=0.99
    linearSpringConstant=550
    worldSpaceMinimumScale=0.1
    boundBone {boneName=cord_01}
    componentAppliedLinearAccClamp=<3000 3000 6000>
    componentLinearAccScale=<0.3 0.3 1>
    componentLinearVelScale=<0.2 0.2 0.1>
    gravityOverride {z=-980}
    physicsBodyDefinitions [{
      boundBone {boneName=cord_01}
      boxExtents=<10 10 10>
      constraintSetup {angularLimitsMax=<0 40 20>, angularLimitsMin=<0 -40 -20>, angularTarget=<1 0 0>}
      localJointOffset=<1 0.1 0.1>
     }]
   }
   n13 = Anim Dynamics - Bone: cord_02(ComponentPose=@n12, Alpha=1, ExternalForce="0,0,0") {
    angularDampingOverride=1
    angularSpringConstant=550
    bAngularSpring=true
    bOverrideAngularDamping=true
    bUseGravityOverride=true
    linearDampingOverride=0.99
    linearSpringConstant=550
    worldSpaceMinimumScale=0.1
    boundBone {boneName=cord_02}
    componentAppliedLinearAccClamp=<3000 3000 6000>
    componentLinearAccScale=<0.3 0.3 1>
    componentLinearVelScale=<0.2 0.2 0.1>
    gravityOverride {z=-980}
    physicsBodyDefinitions [{
      boundBone {boneName=cord_02}
      boxExtents=<10 10 10>
      constraintSetup {angularLimitsMax=<0 20 20>, angularLimitsMin=<0 -20 -20>, angularTarget=<1 0 0>}
      localJointOffset=<1 0.1 0.1>
     }]
   }
   Component To Local(ComponentPose=@n13)
   "Transform (Modify) Bone - Bone: part_down"!(ComponentPose, Alpha=Get My() -> (My:double), Translation="0,0,0", Rotation:Rotator="0,0,-20", Scale:Vector="1,1,1") -> (Pose) {
    rotationMode=BMM_Additive
    rotationSpace=BCS_BoneSpace
    translationMode=BMM_Additive
    translationSpace=BCS_BoneSpace
    boneToModify {boneName=part_down}
    rotation {roll=-20}
   }
   n14 = "Transform (Modify) Bone - Bone: part_down"(
    ComponentPose=Local To Component(LocalPose=Use cached pose 'SavedPose30910'())
    Alpha=1
    Translation="0,0,0.2"
    Rotation="0,0,0"
    Scale="1,1,1"
   ) {translationMode=BMM_Additive, translationSpace=BCS_BoneSpace, boneToModify {boneName=part_down}, translation {z=0.2}}
   n15 = Anim Dynamics - Bone: part_down(
    ComponentPose="Transform (Modify) Bone - Bone: part_up"!(ComponentPose=@n14, Alpha=1, Translation="0,0,1.2", Rotation:Rotator="0,0,0", Scale:Vector="1,1,1") -> (Pose) {translationMode=BMM_Additive, translationSpace=BCS_BoneSpace, boneToModify {boneName=part_up}, translation {z=1.2}}
    Alpha=1
    ExternalForce="0,0,0"
   ) {
    bOverrideLinearDamping=true
    bUseGravityOverride=true
    linearDampingOverride=1
    linearSpringConstant=800
    simulationSpace=RootRelative
    boundBone {boneName=part_down}
    componentLinearAccScale {z=0.05}
    componentLinearVelScale {z=0.05}
    gravityOverride {z=120}
    physicsBodyDefinitions [{
      boundBone {boneName=part_down}
      boxExtents=<10 10 10>
      constraintSetup {angularLimitsMax=<0 0 0>, angularLimitsMin=<0 0 0>, angularTarget=<0 0 0>, linearAxesMin=<0 0 -0.6>}
      localJointOffset=<0 0 0.1>
     }]
   }
   n16 = Anim Dynamics - Bone: part_up(ComponentPose=@n15, Alpha=1, ExternalForce="0,0,0") {
    bLinearSpring=true
    bOverrideLinearDamping=true
    bUseGravityOverride=true
    linearDampingOverride=1
    linearSpringConstant=700
    simulationSpace=RootRelative
    boundBone {boneName=part_up}
    componentAppliedLinearAccClamp=<3003 3000 30000>
    componentLinearAccScale {z=0.4}
    componentLinearVelScale {z=0.3}
    gravityOverride {z=-600}
    physicsBodyDefinitions [{
      boundBone {boneName=part_up}
      boxExtents=<10 10 10>
      constraintSetup {angularLimitsMax=<0 0 0>, angularLimitsMin=<0 0 0>, angularTarget=<0 0 0>, linearAxesMin=<0 0 -1>}
      localJointOffset=<0 0 0.1>
     }]
    simSpaceSettings {maxAngularAcceleration=21, maxAngularVelocity=7}
   }
   Output Pose(Result=Blend Poses by bool(
     BlendPose_0=Use cached pose 'SavedPose30910'()
     BlendPose_1=Component To Local(ComponentPose=@n16)
     BlendTime_0=0.05
     BlendTime_1=0.2
     bActiveValue=Get eliminated() -> (eliminated:bool)
    )) {layerGroup=None}
   n17 = "Transform (Modify) Bone - Bone: pack"(ComponentPose=Local To Component(LocalPose=Input Pose()), Translation="0,2,0", Rotation="0,0,5") {
    rotationMode=BMM_Additive
    translationMode=BMM_Additive
    translationSpace=BCS_BoneSpace
    alphaBoolBlend {blendOption=Linear}
    boneToModify {boneName=pack}
    rotation {roll=5}
    translation {y=2}
   }
   n18 = "Transform (Modify) Bone - Bone: pack"(ComponentPose=@n17, Translation=Get backpack_translation(), Rotation=Get backpack_rotate()) {rotationMode=BMM_Additive, translationMode=BMM_Additive, alphaBoolBlend {blendOption=Linear}, boneToModify {boneName=pack}}
   n19 = "Transform (Modify) Bone - Bone: part_down"(ComponentPose=@n18, Alpha=1, Translation="0,0,0", Rotation="0,0,0", Scale="1,1,1") {rotationMode=BMM_Additive, rotationSpace=BCS_BoneSpace, boneToModify {boneName=part_down}}
   n20 = Anim Dynamics - Bone: body(
    ComponentPose="Transform (Modify) Bone - Bone: body"(
     ComponentPose="Transform (Modify) Bone - Bone: part_up"(ComponentPose=@n19, Alpha=1, Translation="0,0,0", Rotation="0,0,0", Scale="1,1,1") {rotationMode=BMM_Additive, rotationSpace=BCS_BoneSpace, boneToModify {boneName=part_up}}
     Alpha=1
     Translation="0,0,5"
     Rotation="0,0,0"
     Scale="1,1,1"
    ) {translationMode=BMM_Additive, translationSpace=BCS_BoneSpace, boneToModify {boneName=body}, translation {z=5}}
    Alpha=1
    ExternalForce="0,0,0"
   ) {
    angularDampingOverride=0.9999
    bLinearSpring=true
    bOverrideAngularDamping=true
    bOverrideLinearDamping=true
    bUseGravityOverride=true
    linearDampingOverride=0.998
    linearSpringConstant=85
    simulationSpace=BoneRelative
    boundBone {boneName=body}
    componentAppliedLinearAccClamp=<3000 3000 3000>
    componentLinearAccScale {z=0.05}
    componentLinearVelScale {z=0.05}
    gravityOverride {z=-600}
    physicsBodyDefinitions [{
      boundBone {boneName=body}
      boxExtents=<10 10 10>
      constraintSetup {angularLimitsMax=<0 0 0>, angularLimitsMin=<0 0 0>, angularTarget=<0 0 0>, linearAxesMax=<10 0 10>, linearAxesMin=<-10 0 -10>}
      localJointOffset=<0.1 0 0.1>
     }]
    relativeSpaceBone {boneName=root}
    simSpaceSettings {maxAngularAcceleration=21, maxAngularVelocity=9}
   }
   n21 = Anim Dynamics - Bone: cord_01(ComponentPose=@n20, Alpha=1, ExternalForce="0,0,0") {
    angularDampingOverride=1
    angularSpringConstant=550
    bAngularSpring=true
    bOverrideAngularDamping=true
    bUseGravityOverride=true
    linearDampingOverride=0.99
    linearSpringConstant=550
    simulationSpace=BoneRelative
    worldSpaceMinimumScale=0.1
    boundBone {boneName=cord_01}
    componentAppliedLinearAccClamp=<3000 3000 6000>
    componentLinearAccScale=<0.2 0.2 0.5>
    componentLinearVelScale=<0.1 0.1 0.3>
    gravityOverride {z=-980}
    physicsBodyDefinitions [{
      boundBone {boneName=cord_01}
      boxExtents=<10 10 10>
      constraintSetup {angularLimitsMax=<0 20 20>, angularLimitsMin=<0 -20 -5>, angularTarget=<1 0 0>}
      localJointOffset=<1 0 0>
     }]
    relativeSpaceBone {boneName=root}
    simSpaceSettings {maxAngularAcceleration=24, maxAngularVelocity=8}
   }
   n22 = Anim Dynamics - Bone: cord_02(ComponentPose=@n21, Alpha=1, ExternalForce="0,0,0") {
    angularDampingOverride=1
    angularSpringConstant=550
    bAngularSpring=true
    bOverrideAngularDamping=true
    bUseGravityOverride=true
    linearDampingOverride=0.99
    linearSpringConstant=550
    simulationSpace=BoneRelative
    worldSpaceMinimumScale=0.1
    boundBone {boneName=cord_02}
    componentAppliedLinearAccClamp=<3000 3000 6000>
    componentLinearAccScale=<0.2 0.2 0.5>
    componentLinearVelScale=<0.1 0.1 0.3>
    gravityOverride {z=-980}
    physicsBodyDefinitions [{
      boundBone {boneName=cord_02}
      boxExtents=<10 10 10>
      constraintSetup {angularLimitsMax=<0 20 20>, angularLimitsMin=<0 -20 -5>, angularTarget=<1 0 0>}
      localJointOffset=<1 0.1 0.1>
     }]
    relativeSpaceBone {boneName=root}
    simSpaceSettings {maxAngularAcceleration=24, maxAngularVelocity=8}
   }
   Save cached pose 'SavedPose30910'(Pose=Component To Local(ComponentPose=@n22)) {cacheName=SavedPose30910}
  ]}
 EventGraph {nodes [
   n1 = FE_jaw_move() {is=title K2Node_CustomEvent}
   Set WhenStartPlay(execute=[@n1.then, @n11.then], WhenStartPlay:double=GetGameTimeInSeconds()) -> (Output_Get:double)
   > Set eliminated(eliminated:bool=true) -> (Output_Get:bool)
   > RetriggerableDelay(Duration=3.2)
   > Set eliminated(eliminated=false)
   n2 = Event BlueprintUpdateAnimation()
   > Set Delta Time X(Delta Time X=@n2.DeltaTimeX)
   > n3 = Sequence!(execute) -> (then_0, then_1, then_2:Exec)
   n4 = GetGameTimeInSeconds()
   n5 = Get WhenStartPlay() -> (WhenStartPlay:double)
   n6 = GetRuntimeFloatCurveValue!(
    Curve:"Runtime Float Curve Structure (by ref)"=Get Curve() -> (Curve:Runtime Float Curve Structure)
    InTime:float=float * float(A=float - float(A=@n4, B=@n5), B=1.8) {is=title K2Node_PromotableOperator}
    InDefaultValue:float=0
   ) -> (ReturnValue:float)
   Set!(execute=@n3.then_1, JawRot_Roll:float=float * float(A=@n6, B=-15), JawRot_Pitch:float=0, JawRot_Yaw:float=0) -> (then, Output_Get:Rotator)
   > Set!(execute, HeadRot_Roll:float=float * float(A=@n6, B=15), HeadRot_Pitch:float=0, HeadRot_Yaw:float=0) -> (then, Output_Get)
   > Branch(Condition=AND Boolean(A=Get play_anim() -> (play_anim:bool), B=Get bIsInFrontEnd()))
   > FE_jaw_move() {is=title K2Node_CallFunction}
   > Set play_anim(play_anim:bool=false) -> (Output_Get:bool)
   > Delay(Duration=3)
   > Set play_anim(play_anim=true)
   n7 = GetOwnerSkeletalMeshComponent()
   n8 = BreakRotator(InRot=BreakTransform(InTransform=GetDeltaTransformFromRefPose(self=@n7, BoneName=spine_05, BaseName=spine_01)).Rotation)
   n9 = MapRangeClamped(Value=@n8.Yaw, InRangeA=-10, InRangeB=-50, OutRangeB=35)
   Set backpack_rotate(execute=@n3.then_0, backpack_rotate=MakeRotator(Roll=@n9, Pitch=float / float(A=@n9, B=-10)))
   > Set backpack_translation(backpack_translation=MakeVector(Y=MapRangeClamped(Value=@n8.Yaw, InRangeA=-15, InRangeB=-50, OutRangeB=-15), Z=-6.5))
   n10 = Event BlueprintProcessAnimInputEvent()
   > n11 = Branch(Condition=IsNamed(self=@n10.AnimInputEvent, InEventName=aie.EnemyPlayerEliminated))
   Branch(Condition=Get eliminated())
   > Set Clap_Timer(Clap_Timer:double=float + float(A=Get Clap_Timer() -> (Clap_Timer:double), B=1) {is=title K2Node_CommutativeAssociativeBinaryOperator}) -> (Output_Get:double)
   n12 = float * float(A=GetGameTimeInSeconds(), B=750) {is=title K2Node_CommutativeAssociativeBinaryOperator}
   n13 = "Absolute (Float)"(A=float * float(A="Sin (Degrees)"!(A:double=@n12) -> (ReturnValue:double), B=20))
   > Set Jaw Laugh Rot(Jaw Laugh Rot:Rotator=MakeRotator(Roll=float * float(A=@n13, B=-1))) -> (Output_Get:Rotator)
   n14 = float * float(A=GetGameTimeInSeconds(), B=750)
   n15 = "Absolute (Float)"(A=float * float(A="Sin (Degrees)"(A=@n14), B=20))
   > Set Jaw Laugh Rot 2(Jaw Laugh Rot 2:Rotator=MakeRotator(Roll=float * float(A=@n15, B=1))) -> (Output_Get:Rotator)
   n16 = Get Clap_Timer()
   n17 = Get Clap_Timer_Length() -> (Clap_Timer_Length:double)
   > Set Clap Alpha(Clap Alpha:double=MapRangeClamped(Value=float - float(A=@n17, B=@n16) {is=title K2Node_CallFunction}, InRangeA=Get Clap_Timer_Length(), InRangeB=0, OutRangeA=1, OutRangeB=0)) -> (Output_Get:double)
   > Branch(Condition="Equal (Float)"!(A:double=Get Clap Alpha(), B:double=0) -> (ReturnValue:bool) {is=title K2Node_CallFunction})
   > Set eliminated(eliminated=false)
   > Set Jaw Laugh Rot(Jaw Laugh Rot="0, 0, 0")
   > Set Jaw Laugh Rot 2(Jaw Laugh Rot 2="0, 0, 0")
   > Set Clap_Timer(Clap_Timer=0)
   > Delay(Duration=0.01)
   > Set Clap Alpha(Clap Alpha=1)
   Set eliminated(eliminated=IsNamed(InEventName=aie.EnemyPlayerEliminated))
   > Delay(Duration=0.8)
   > Set eliminated(eliminated=false)
   n18 = GetRealTimeSeconds()
   n19 = "Sin (Radians)"(A=float + float(A=float * float(A=@n18, B=12) {is=title K2Node_PromotableOperator}) {is=title K2Node_PromotableOperator})
   Set body move_up(body move_up:Vector=MakeVector(Y=0, Z=float * float(A=@n19, B=0.25))) -> (Output_Get:Vector)
   n20 = "Sin (Radians)"(A=float + float(A=float * float(A=@n18, B=12)))
   > Set body move_down(body move_down:Vector=MakeVector(Y=0, Z=float * float(A=@n20, B=-0.25))) -> (Output_Get:Vector)
   TryGetPawnOwner()
   Set eliminated(eliminated=IsNamed(InEventName=aie.EnemyPlayerEliminated))
   n21 = Get eliminated()
   n22 = Get Delta Time X()
   Set My(My:double=FInterpTo(Current=Get My(), Target=SelectFloat(A=1, B=0, bPickA=@n21), DeltaTime=float / float(A=@n22, B=2), InterpSpeed=10)) -> (Output_Get:double)
  ]}
]
variables [
 backpack_rotate {type=Rotator}
 backpack_translation {type=Vector}
 bEnemyPlayerEliminated {type=number}
 body move_down {type=Vector}
 body move_up {type=Vector}
 Clap Alpha {type=number}
 Clap_Timer {type=number}
 Clap_Timer_Length {type=number, default=60}
 Curve {
  type=RuntimeFloatCurve
  default {
   externalCurve=null
   editorCurveData {
    defaultValue=3.4028234663852886e+38
    postInfinityExtrap=RCCE_Constant
    preInfinityExtrap=RCCE_Constant
    keys [
     {
      arriveTangent=0
      arriveTangentWeight=0
      interpMode=RCIM_Linear
      leaveTangent=0
      leaveTangentWeight=0
      tangentMode=RCTM_Auto
      tangentWeightMode=RCTWM_WeightedNone
      time=0.026
      value=0
     }
     {
      arriveTangent=2.8722
      arriveTangentWeight=0
      interpMode=RCIM_Cubic
      leaveTangent=2.8722
      leaveTangentWeight=0
      tangentMode=RCTM_User
      tangentWeightMode=RCTWM_WeightedNone
      time=0.0518
      value=0.5831
     }
     {
      arriveTangent=-4.1257
      arriveTangentWeight=0
      interpMode=RCIM_Cubic
      leaveTangent=-4.1257
      leaveTangentWeight=0
      tangentMode=RCTM_Auto
      tangentWeightMode=RCTWM_WeightedNone
      time=0.1704
      value=0.7225
     }
     {
      arriveTangent=0
      arriveTangentWeight=0
      interpMode=RCIM_Linear
      leaveTangent=0
      leaveTangentWeight=0
      tangentMode=RCTM_Auto
      tangentWeightMode=RCTWM_WeightedNone
      time=0.1931
      value=0
     }
     {
      arriveTangent=0
      arriveTangentWeight=0
      interpMode=RCIM_Linear
      leaveTangent=0
      leaveTangentWeight=0
      tangentMode=RCTM_Auto
      tangentWeightMode=RCTWM_WeightedNone
      time=0.3383
      value=0
     }
     {
      arriveTangent=2.9236
      arriveTangentWeight=0
      interpMode=RCIM_Cubic
      leaveTangent=2.9236
      leaveTangentWeight=0
      tangentMode=RCTM_User
      tangentWeightMode=RCTWM_WeightedNone
      time=0.4312
      value=0.5353
     }
     {
      arriveTangent=-0.8575
      arriveTangentWeight=0
      interpMode=RCIM_Cubic
      leaveTangent=-0.8575
      leaveTangentWeight=0
      tangentMode=RCTM_User
      tangentWeightMode=RCTWM_WeightedNone
      time=0.5487
      value=0.7003
     }
     {
      arriveTangent=0
      arriveTangentWeight=0
      interpMode=RCIM_Linear
      leaveTangent=0
      leaveTangentWeight=0
      tangentMode=RCTM_Auto
      tangentWeightMode=RCTWM_WeightedNone
      time=0.6246
      value=0
     }
     {
      arriveTangent=0.0048
      arriveTangentWeight=0
      interpMode=RCIM_Cubic
      leaveTangent=0.0048
      leaveTangentWeight=0
      tangentMode=RCTM_User
      tangentWeightMode=RCTWM_WeightedNone
      time=0.7672
      value=0
     }
     {
      arriveTangent=2.1278
      arriveTangentWeight=0
      interpMode=RCIM_Cubic
      leaveTangent=2.1278
      leaveTangentWeight=0
      tangentMode=RCTM_User
      tangentWeightMode=RCTWM_WeightedNone
      time=1.0603
      value=0.6108
     }
     {
      arriveTangent=0.0261
      arriveTangentWeight=0
      interpMode=RCIM_Cubic
      leaveTangent=-6.6124
      leaveTangentWeight=0
      tangentMode=RCTM_Break
      tangentWeightMode=RCTWM_WeightedNone
      time=1.444
      value=0.865
     }
     {
      arriveTangent=-9.9309
      arriveTangentWeight=0
      interpMode=RCIM_Linear
      leaveTangent=-9.9309
      leaveTangentWeight=0
      tangentMode=RCTM_Auto
      tangentWeightMode=RCTWM_WeightedNone
      time=1.5064
      value=0
     }
    ]
   }
  }
 }
 Delta Time X {type=number}
 eliminated {type=boolean}
 HeadRot {type=Rotator}
 Jaw Laugh Rot {type=Rotator}
 Jaw Laugh Rot 2 {type=Rotator}
 JawRot {type=Rotator}
 My {type=number}
 NewVar {type=number}
 play_anim {type=boolean, default=true}
 rot_part {type=Rotator}
 WhenStartPlay {type=number}
]
